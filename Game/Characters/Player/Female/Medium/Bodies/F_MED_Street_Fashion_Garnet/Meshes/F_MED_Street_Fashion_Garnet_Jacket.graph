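# Physics asset for the Female Medium Street Fashion Garnet player jacket, setting up bodies for its dynamic bones to move as secondary motion.
# Seven chains around the jacket attach to a spine_03 torso capsule through tiny master capsules with free swing and twist.
# Thin box panels hang from these masters, two segments at each front and one elsewhere.
# Panel joints have narrow twist limits, with tighter swing across the three back chains.
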PhysicsAsset /Game/Characters/Player/Female/Medium/Bodies/F_MED_Street_Fashion_Garnet/Meshes/F_MED_Street_Fashion_Garnet_Jacket
bodies {
 dyn_jacket_bk_01 [{shapeType=Box, extentX=9, extentY=9.5, extentZ=0.6903, center=<3.9836 0.0001 0>}]
 dyn_jacket_bk_MSTR [{shapeType=Capsule, length=0.5, radius=0.5, rotation=<rot 90 0 0>}]
 dyn_jacket_bk_l_01 [{shapeType=Box, extentX=9, extentY=9.5, extentZ=0.6903, center=<4.0261 0.0001 0>}]
 dyn_jacket_bk_l_MSTR [{shapeType=Capsule, length=0.5, radius=0.5, rotation=<rot 90 0 0>}]
 dyn_jacket_bk_r_01 [{shapeType=Box, extentX=9, extentY=9.5, extentZ=0.6903, center=<-4.124 0 0>}]
 dyn_jacket_bk_r_MSTR [{shapeType=Capsule, length=0.5, radius=0.5, rotation=<rot 90 0 0>}]
 dyn_jacket_fr_l_01 [{shapeType=Box, extentX=7.1362, extentY=9.5, extentZ=0.6903, center=<2.6448 0.0001 0>}]
 dyn_jacket_fr_l_02 [{shapeType=Box, extentX=5.8562, extentY=9.5, extentZ=0.6903, center=<2.0136 0.0001 0>}]
 dyn_jacket_fr_l_MSTR [{shapeType=Capsule, length=0.5, radius=0.5, rotation=<rot 90 0 0>}]
 dyn_jacket_fr_r_01 [{shapeType=Box, extentX=7.1752, extentY=9.5, extentZ=1.1903, center=<-2.5092 0 0>}]
 dyn_jacket_fr_r_02 [{shapeType=Box, extentX=4.9334, extentY=9.5, extentZ=1.1903, center=<-1.7355 0 0>}]
 dyn_jacket_fr_r_MSTR [{shapeType=Capsule, length=0.5, radius=0.5, rotation=<rot 90 0 0>}]
 dyn_jacket_l_01 [{shapeType=Box, extentX=9, extentY=9.5, extentZ=0.6903, center=<3.6782 0.0001 0>}]
 dyn_jacket_l_MSTR [{shapeType=Capsule, length=0.5, radius=0.5, rotation=<rot 90 0 0>}]
 dyn_jacket_r_01 [{shapeType=Box, extentX=9, extentY=9.5, extentZ=1.1903, center=<-3.6782 0 0>}]
 dyn_jacket_r_MSTR [{shapeType=Capsule, length=0.5, radius=0.5, rotation=<rot 90 0 0>}]
 spine_03 [{shapeType=Capsule, length=3.2426, radius=8.7172, center=<-2.0188 2.4863 0>}]
}
constraints [
 {
  bone1Name=dyn_jacket_fr_l_MSTR
  bone2Name=spine_03
  swing1LimitDegrees=45
  swing1Motion=Free
  swing2LimitDegrees=45
  swing2Motion=Free
  twistLimitDegrees=45
  twistMotion=Free
 }
 {bone1Name=dyn_jacket_fr_l_01, bone2Name=dyn_jacket_fr_l_MSTR, swing1LimitDegrees=45, swing2LimitDegrees=45, twistLimitDegrees=5}
 {bone1Name=dyn_jacket_fr_l_02, bone2Name=dyn_jacket_fr_l_01, swing1LimitDegrees=45, swing2LimitDegrees=45, twistLimitDegrees=5}
 {
  bone1Name=dyn_jacket_l_MSTR
  bone2Name=spine_03
  swing1LimitDegrees=45
  swing1Motion=Free
  swing2LimitDegrees=45
  swing2Motion=Free
  twistLimitDegrees=45
  twistMotion=Free
 }
 {bone1Name=dyn_jacket_l_01, bone2Name=dyn_jacket_l_MSTR, swing1LimitDegrees=45, swing2LimitDegrees=45, twistLimitDegrees=5}
 {
  bone1Name=dyn_jacket_bk_l_MSTR
  bone2Name=spine_03
  swing1LimitDegrees=45
  swing1Motion=Free
  swing2LimitDegrees=45
  swing2Motion=Free
  twistLimitDegrees=45
  twistMotion=Free
 }
 {bone1Name=dyn_jacket_bk_l_01, bone2Name=dyn_jacket_bk_l_MSTR, swing1LimitDegrees=20, swing2LimitDegrees=30, twistLimitDegrees=5}
 {
  bone1Name=dyn_jacket_bk_MSTR
  bone2Name=spine_03
  swing1LimitDegrees=45
  swing1Motion=Free
  swing2LimitDegrees=45
  swing2Motion=Free
  twistLimitDegrees=45
  twistMotion=Free
 }
 {bone1Name=dyn_jacket_bk_01, bone2Name=dyn_jacket_bk_MSTR, swing1LimitDegrees=25, swing2LimitDegrees=30, twistLimitDegrees=5}
 {
  bone1Name=dyn_jacket_bk_r_MSTR
  bone2Name=spine_03
  swing1LimitDegrees=45
  swing1Motion=Free
  swing2LimitDegrees=45
  swing2Motion=Free
  twistLimitDegrees=45
  twistMotion=Free
 }
 {bone1Name=dyn_jacket_bk_r_01, bone2Name=dyn_jacket_bk_r_MSTR, swing1LimitDegrees=20, swing2LimitDegrees=30, twistLimitDegrees=5}
 {
  bone1Name=dyn_jacket_r_MSTR
  bone2Name=spine_03
  swing1LimitDegrees=45
  swing1Motion=Free
  swing2LimitDegrees=45
  swing2Motion=Free
  twistLimitDegrees=45
  twistMotion=Free
 }
 {bone1Name=dyn_jacket_r_01, bone2Name=dyn_jacket_r_MSTR, swing1LimitDegrees=45, swing2LimitDegrees=45, twistLimitDegrees=5}
 {
  bone1Name=dyn_jacket_fr_r_MSTR
  bone2Name=spine_03
  swing1LimitDegrees=45
  swing1Motion=Free
  swing2LimitDegrees=45
  swing2Motion=Free
  twistLimitDegrees=45
  twistMotion=Free
 }
 {bone1Name=dyn_jacket_fr_r_01, bone2Name=dyn_jacket_fr_r_MSTR, swing1LimitDegrees=45, swing2LimitDegrees=45, twistLimitDegrees=5}
 {bone1Name=dyn_jacket_fr_r_02, bone2Name=dyn_jacket_fr_r_01, swing1LimitDegrees=45, swing2LimitDegrees=45, twistLimitDegrees=5}
]
