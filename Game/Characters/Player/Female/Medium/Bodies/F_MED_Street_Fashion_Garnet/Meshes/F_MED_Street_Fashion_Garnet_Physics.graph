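# Physics asset for the Street Fashion Garnet female medium body, covering its hood and left and right lace dangles.
# Capsules on spine_04 and spine_05 form the torso, with the hood tightly limited against spine_05.
# Each lace is a three-link capsule chain hanging from a small parent body, which has free swing and twist on spine_05.
# Links use narrow twist and moderate swing limits, tighter on the upper right side.
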PhysicsAsset /Game/Characters/Player/Female/Medium/Bodies/F_MED_Street_Fashion_Garnet/Meshes/F_MED_Street_Fashion_Garnet_Physics
bodies {
 dyn_hood_01 [{shapeType=Capsule, length=9.5837, radius=4.8689, center=<7.3835 0.8421 -0.2798>, rotation=<rot -2.9922 -90 1.9843>}]
 dyn_lace_01_l [{shapeType=Capsule, length=2, radius=1, center=<2 0.5692 0>, rotation=<rot 90 0 0>}]
 dyn_lace_01_r [{shapeType=Capsule, length=2, radius=1.5, center=<0.5 1.12 0>, rotation=<rot 90 0 0>}]
 dyn_lace_02_l [{shapeType=Capsule, length=2.25, radius=0.75, center=<2.5 0.37 0>, rotation=<rot 90 0 0>}]
 dyn_lace_02_r [{shapeType=Capsule, length=2.25, radius=0.75, center=<0.7 0.525 0>, rotation=<rot 90 0 0>}]
 dyn_lace_03_l [{shapeType=Capsule, length=1.25, radius=0.75, center=<2.8219 0.3859 0>, rotation=<rot 90 0 0>}]
 dyn_lace_03_r [{shapeType=Capsule, length=1.5, radius=0.75, center=<1 0.6 0>, rotation=<rot 90 0 0>}]
 dyn_parent_lace_01_l [{shapeType=Capsule, length=0.505, radius=0.505, rotation=<rot 0 -90 0>}]
 dyn_parent_lace_01_r [{shapeType=Capsule, length=0.505, radius=0.505, rotation=<rot 0 -90 0>}]
 spine_04 [{shapeType=Capsule, length=12.8352, radius=12.9844, center=<-2.4988 1.8032 0.1535>, rotation=<rot 0 -90 0>}]
 spine_05 [{shapeType=Capsule, length=10.75, radius=15.75, center=<-5.3233 -4.5696 0>, rotation=<rot 0 -26.2143 0>}]
}
constraints [
 {bone1Name=dyn_lace_01_l, bone2Name=dyn_parent_lace_01_l, swing1LimitDegrees=30, swing2LimitDegrees=30, twistLimitDegrees=8}
 {bone1Name=dyn_lace_02_l, bone2Name=dyn_lace_01_l, swing1LimitDegrees=30, swing2LimitDegrees=30, twistLimitDegrees=8}
 {bone1Name=dyn_lace_03_l, bone2Name=dyn_lace_02_l, swing1LimitDegrees=30, swing2LimitDegrees=30, twistLimitDegrees=8}
 {bone1Name=dyn_lace_01_r, bone2Name=dyn_parent_lace_01_r, swing1LimitDegrees=15, swing2LimitDegrees=15, twistLimitDegrees=8}
 {bone1Name=dyn_lace_02_r, bone2Name=dyn_lace_01_r, swing1LimitDegrees=18, swing2LimitDegrees=25, twistLimitDegrees=8}
 {bone1Name=dyn_lace_03_r, bone2Name=dyn_lace_02_r, swing1LimitDegrees=30, swing2LimitDegrees=30, twistLimitDegrees=8}
 {
  bone1Name=dyn_parent_lace_01_l
  bone2Name=spine_05
  swing1LimitDegrees=45
  swing1Motion=Free
  swing2LimitDegrees=45
  swing2Motion=Free
  twistLimitDegrees=45
  twistMotion=Free
 }
 {
  bone1Name=dyn_parent_lace_01_r
  bone2Name=spine_05
  swing1LimitDegrees=45
  swing1Motion=Free
  swing2LimitDegrees=45
  swing2Motion=Free
  twistLimitDegrees=45
  twistMotion=Free
 }
 {
  bone1Name=spine_05
  bone2Name=spine_04
  swing1LimitDegrees=45
  swing1Motion=Free
  swing2LimitDegrees=45
  swing2Motion=Free
  twistLimitDegrees=45
  twistMotion=Free
 }
 {bone1Name=dyn_hood_01, bone2Name=spine_05, swing1LimitDegrees=10, swing2LimitDegrees=5, twistLimitDegrees=5}
]
