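# Secondary-motion AnimBlueprint for the Street Fashion Garnet medium female player body, copying its pose from the attached parent mesh.
# Each update derives per-side thigh-versus-pelvis rotation alphas that drive additive corrective transforms on the glute bones.
# Weighted multi-bone constraints then position lace and boot parent bones.
# Two chained RigidBody nodes simulate the main and jacket physics assets in spine_05 space, with overridden gravity and damped component motion.
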AnimBlueprint /Game/Characters/Player/Female/Medium/Bodies/F_MED_Street_Fashion_Garnet/Meshes/F_MED_Street_Fashion_Garnet_AnimBP
target_skeleton=&/F_MED_Street_Fashion_Garnet_Skeleton
graphs [
 AnimGraph {nodes [
   n1 = "Transform (Modify) Bone - Bone: dyn_glute_l"(
    ComponentPose=Local To Component(LocalPose=Copy Pose From Mesh() {bUseAttachedParent=true})
    Alpha=Get thigh_rot_l()
    Translation="-4.689,5.014,-4.028"
    Rotation="0.000,0.000,-55"
    Scale="1.25,1.25,1.25"
   ) {
    rotationMode=BMM_Additive
    rotationSpace=BCS_BoneSpace
    scaleMode=BMM_Additive
    scaleSpace=BCS_BoneSpace
    translationMode=BMM_Additive
    translationSpace=BCS_BoneSpace
    alphaBoolBlend {blendOption=Linear}
    boneToModify {boneName=dyn_glute_l}
    rotation {roll=-55}
    scale=<1.25 1.25 1.25>
    translation=<-4.689 5.014 -4.028>
   }
   n2 = "Transform (Modify) Bone - Bone: dyn_glute_r"(ComponentPose=@n1, Alpha=Get thigh_rot_r(), Translation="-3.75,0,0", Rotation="0.000,0.000,-55", Scale="1.3,1.3,1.3") {
    rotationMode=BMM_Additive
    rotationSpace=BCS_BoneSpace
    scaleMode=BMM_Additive
    scaleSpace=BCS_BoneSpace
    translationMode=BMM_Additive
    translationSpace=BCS_BoneSpace
    alphaBoolBlend {blendOption=Linear}
    boneToModify {boneName=dyn_glute_r}
    rotation {roll=-55}
    scale=<1.3 1.3 1.3>
    translation {x=-3.75}
   }
   n3 = "Constraint - dyn_parent_lace_01_l (PPPPPPP)"!(
    ComponentPose=@n2
    Alpha=1
    ConstraintWeights_0=0.019
    ConstraintWeights_1:float=0.1474
    ConstraintWeights_2:float=0.0735
    ConstraintWeights_3:float=0.6682
    ConstraintWeights_4:float=0.028
    ConstraintWeights_5:float=0.0405
    ConstraintWeights_6:float=0.0235
   ) -> (Pose) {
    alphaBoolBlend {blendOption=Linear}
    boneToModify {boneName=dyn_parent_lace_01_l}
    constraintSetup [
     {targetBone {boneName=head}}
     {targetBone {boneName=clavicle_l}}
     {targetBone {boneName=spine_04}}
     {targetBone {boneName=spine_05}}
     {targetBone {boneName=upperarm_twist_01_l}}
     {targetBone {boneName=neck_02}}
     {targetBone {boneName=neck_01}}
    ]
    constraintWeights [1,1,1,1,1,1,1]
   }
   n4 = "Constraint - dyn_parent_lace_01_r (PPPPPPP)"!(
    ComponentPose=@n3
    Alpha=1
    ConstraintWeights_0=0.0226
    ConstraintWeights_1:float=0.0289
    ConstraintWeights_2:float=0.0637
    ConstraintWeights_3:float=0.6546
    ConstraintWeights_4:float=0.1547
    ConstraintWeights_5:float=0.0479
    ConstraintWeights_6:float=0.0276
   ) -> (Pose) {
    alphaBoolBlend {blendOption=Linear}
    boneToModify {boneName=dyn_parent_lace_01_r}
    constraintSetup [
     {targetBone {boneName=head}}
     {targetBone {boneName=upperarm_twist_01_r}}
     {targetBone {boneName=spine_04}}
     {targetBone {boneName=spine_05}}
     {targetBone {boneName=clavicle_r}}
     {targetBone {boneName=neck_02}}
     {targetBone {boneName=neck_01}}
    ]
    constraintWeights [1,1,1,1,1,1,1]
   }
   n5 = "Constraint - dyn_parent_boot_r (PPP)"!(ComponentPose=@n4, Alpha=1, ConstraintWeights_0=0.2825, ConstraintWeights_1:float=0.6161, ConstraintWeights_2:float=0.1014) -> (Pose) {
    alphaBoolBlend {blendOption=Linear}
    boneToModify {boneName=dyn_parent_boot_r}
    constraintSetup [{targetBone {boneName=calf_twist_02_r}}, {targetBone {boneName=calf_twist_01_r}}, {targetBone {boneName=calf_r}}]
    constraintWeights [1,1,1]
   }
   n6 = "Constraint - dyn_parent_boot_l (PPP)"!(ComponentPose=@n5, Alpha=1, ConstraintWeights_0=0.1012, ConstraintWeights_1:float=0.6285, ConstraintWeights_2:float=0.2703) -> (Pose) {
    alphaBoolBlend {blendOption=Linear}
    boneToModify {boneName=dyn_parent_boot_l}
    constraintSetup [{targetBone {boneName=calf_l}}, {targetBone {boneName=calf_twist_01_l}}, {targetBone {boneName=calf_twist_02_l}}]
    constraintWeights [1,1,1]
   }
   n7 = RigidBody(ComponentPose=@n6) {
    bClampLinearTranslationLimitToRefPose=true
    bOverrideWorldGravity=true
    overridePhysicsAsset=&/F_MED_Street_Fashion_Garnet_Physics
    simulationSpace=BaseBoneSpace
    alphaBoolBlend {blendOption=Linear}
    baseBoneRef {boneName=spine_05}
    componentAppliedLinearAccClamp=<3000 3000 1000>
    componentLinearAccScale=<0.35 0.6 0.2>
    componentLinearVelScale=<0.35 0.6 0.2>
    overrideWorldGravity {z=-980}
   }
   n8 = RigidBody(ComponentPose=@n7) {
    bClampLinearTranslationLimitToRefPose=true
    bOverrideWorldGravity=true
    overridePhysicsAsset=&/F_MED_Street_Fashion_Garnet_Jacket
    simulationSpace=BaseBoneSpace
    alphaBoolBlend {blendOption=Linear}
    baseBoneRef {boneName=spine_05}
    componentAppliedLinearAccClamp=<3000 3000 1500>
    componentLinearAccScale=<0.4 0.4 0.2>
    componentLinearVelScale=<0.4 0.4 0.2>
    overrideWorldGravity {z=-980}
   }
   Output Pose(Result=Component To Local(ComponentPose=@n8)) {layerGroup=None}
  ]}
 EventGraph {nodes [
   Event BlueprintUpdateAnimation()
   > Get!(execute:Exec) -> (then:Exec, else:Exec, Skel_Mesh:Skeletal Mesh Component Object Reference)
   > Sequence()
   n1 = Get Skel_Mesh()
   > Set thigh_rot_l(thigh_rot_l=GetAlphaFromReferenceTransformDelta(SkeletalMeshComponent=@n1, BoneName=thigh_l, BaseName=pelvis, InputRange=(X=-20,Y=-100), TwistAxis="1,0,0"))
   > Set thigh_rot_r(thigh_rot_r=GetAlphaFromReferenceTransformDelta(SkeletalMeshComponent=@n1, BoneName=thigh_r, BaseName=pelvis, InputRange=(X=-20,Y=-100), TwistAxis="1,0,0"))
   Event BlueprintInitializeAnimation()
   > Set Skel_Mesh(Skel_Mesh=GetOwnerSkeletalMeshComponent())
  ]}
]
variables [
 Skel_Mesh {type=SkeletalMeshComponent}
 Skeletal_Mesh {type=SkeletalMeshComponent}
 thigh_rot_l {type=number}
 thigh_rot_r {type=number}
]
